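; Light beam actor for LimeRock that is shown or hidden through a replicated on/off flag.
; On BeginPlay, if set to follow calendar events, the authority waits for match ready and calendar sync, then turns it on if the watched event is active.
; PC_TurnOnLights and PC_TurnOffLights also drive the flag.
; OnRep_TurnOnClient toggles actor visibility where cosmetic events can execute, with debug prints along the path.
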(event Custom|PC_TurnOffLights
  (CallFunction|TurnOffLight))

(event Custom|PC_TurnOnLights
  (CallFunction|TurnonLight))

(event EventBeginPlay
  (Networking|SwitchHasAuthority
    (:Authority
      (if (Variables|Calendar|GetFollowCalendarEvents)
        (Class|FortAsyncActionAuthorityMatchReady|MatchReadyAsync
          (:OnReady
            (MCPContext|InitialCalendarSyncCompleted
              (:OnSyncComplete
                (bind _biseventactive (CalendarEvents|GetCalendarEventInformation (Variables|Calendar|GetCalendarEventToWatch)))
                (if _biseventactive
                  (CallFunction|TurnonLight))))))))))

(event Custom|TurnOnLight
  (Development|PrintString "Turn on the light!")
  (Variables|Replication|SetTurnOnClient true))

(event Custom|TurnOffLight
  (Development|PrintString "Turn off the light!")
  (Variables|Replication|SetTurnOnClient false))

(fn OnRep_TurnOnClient
  (Networking|CanExecuteCosmeticEvents
    (:True
      (if (Variables|Replication|GetTurnOnClient)
        (Rendering|SetActorHiddenInGame false)
        (Development|PrintString "Client Side: Light Active")
        (else
          (Rendering|SetActorHiddenInGame true)
          (Development|PrintString "Client Side: Light Inactive"))))))

(fn ConstructionScript)
